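# Conversation database for the NPC Talon Pane's "Ask about..." dialogue, offering lore and advice on O.X.R. and the Swarm.
# The entry tag leads to a question hub with choices for O.X.R., Disappearance of the Swarm, Good Advice, and a Back node.
# Each topic runs as a linear chain of Speech nodes, and all three converge on one shared ReturnToStart node labeled BACK.
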ConversationDatabase /NPCLibrary/NPCs/TalonPane/NPCLibrary_Conversation_TalonPane_Dialogue
entry_points [{tag=Conversation.Entry.TalonPane.Dialogue, destinations [#1]}]
nodes [
 ConversationEntryPointNode_7 {
  id=#1
  class=ConversationEntryPointNode
  outputs [#2]
  properties {entryTag {tagName=Conversation.Entry.TalonPane.Dialogue}}
 }
 Speech_C_0 {
  id=#2
  class=Speech_C
  outputs [#3, #4, #5, #6]
  properties {nodeName="", softTaskIcon=&T_UI_NPC_question, generalConfig {defaultMessage=You've got questions?}}
  sub_nodes [ConversationChoiceNode_5 {class=ConversationChoiceNode, properties {defaultChoiceDisplayText=Ask about...}}]
 }
 Speech_C_6 {
  id=#6
  class=Speech_C
  outputs [#7]
  properties {
   nodeName=""
   softTaskIcon=&T_UI_NPC_question
   generalConfig {defaultMessage=O.X.R. doesn't look like your typical army. Know why?}
  }
  sub_nodes [ConversationChoiceNode_1 {class=ConversationChoiceNode, properties {defaultChoiceDisplayText=Good Advice}}]
 }
 Speech_C_8 {
  id=#7
  class=Speech_C
  outputs [#8]
  properties {
   nodeName=""
   softTaskIcon=&T_UI_NPC_question
   generalConfig {defaultMessage=We all play to our strengths and stay true to ourselves.}
  }
 }
 Speech_C_9 {
  id=#8
  class=Speech_C
  outputs [#9]
  properties {
   nodeName=""
   softTaskIcon=&T_UI_NPC_question
   generalConfig {defaultMessage="On that battlefield, every Bug is the same. But only you can be you."}
  }
 }
 Speech_C_10 {
  id=#9
  class=Speech_C
  outputs [#10]
  properties {
   nodeName=""
   softTaskIcon=&T_UI_NPC_question
   generalConfig {defaultMessage="Oh, and... don't rush into battle without some ammo. You'll need that."}
  }
 }
 ReturnToStart_C_23 {
  id=#10
  class=ReturnToStart_C
  properties {softTaskIcon=&T_UI_NPC_Back}
  sub_nodes [ConversationChoiceNode_32 {class=ConversationChoiceNode, properties {defaultChoiceDisplayText=BACK}}]
 }
 Speech_C_2 {
  id=#5
  class=Speech_C
  outputs [#11]
  properties {nodeName="", softTaskIcon=&T_UI_NPC_question, generalConfig {defaultMessage=Swarm may be gone...}}
  sub_nodes [ConversationChoiceNode_3 {class=ConversationChoiceNode, properties {defaultChoiceDisplayText=Disappearance of the Swarm}}]
 }
 Speech_C_3 {
  id=#11
  class=Speech_C
  outputs [#12]
  properties {
   nodeName=""
   softTaskIcon=&T_UI_NPC_question
   generalConfig {defaultMessage=...but I've been fighting long enough to know a calm before the storm when I'm in one.}
  }
 }
 Speech_C_1 {
  id=#12
  class=Speech_C
  outputs [#13]
  properties {
   nodeName=""
   softTaskIcon=&T_UI_NPC_question
   generalConfig {defaultMessage=That blue-haired kid and her dad are right. Something's comin'.}
  }
 }
 Speech_C_11 {
  id=#13
  class=Speech_C
  outputs [#10]
  properties {
   nodeName=""
   softTaskIcon=&T_UI_NPC_question
   generalConfig {defaultMessage="Stay braced, soldier. We gotta be ready for it."}
  }
 }
 Back_C_1 {
  id=#4
  class=Back_C
  sub_nodes [ConversationChoiceNode_2 {class=ConversationChoiceNode, properties {defaultChoiceDisplayText=BACK}}]
 }
 Speech_C_7 {
  id=#3
  class=Speech_C
  outputs [#14]
  properties {
   nodeName=""
   softTaskIcon=&T_UI_NPC_question
   generalConfig {defaultMessage="The greatest fighting force in existence, soldier."}
  }
  sub_nodes [ConversationChoiceNode_0 {class=ConversationChoiceNode, properties {defaultChoiceDisplayText=O.X.R.}}]
 }
 Speech_C_15 {
  id=#14
  class=Speech_C
  outputs [#15]
  properties {
   nodeName=""
   softTaskIcon=&T_UI_NPC_question
   generalConfig {defaultMessage=We can take on any threat. You know why?}
  }
 }
 Speech_C_14 {
  id=#15
  class=Speech_C
  outputs [#16]
  properties {
   nodeName=""
   softTaskIcon=&T_UI_NPC_question
   generalConfig {defaultMessage="Because we've got each other, and we never give up."}
  }
 }
 Speech_C_5 {
  id=#16
  class=Speech_C
  outputs [#10]
  properties {
   nodeName=""
   softTaskIcon=&T_UI_NPC_question
   generalConfig {defaultMessage="Also our weapons are really, really, REALLY good."}
  }
 }
]
